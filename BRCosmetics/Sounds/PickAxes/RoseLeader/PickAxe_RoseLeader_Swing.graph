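# Sound cue for the RoseLeader cosmetic pickaxe swing, limited per owner with stop-oldest concurrency.
# A quality-level root selects mixes that shrink the variation pool at lower tiers, with the last two slots sharing the smallest mix.
# Each mix layers RoseLeader swing waves under close attenuation with generic Fire Axe distant swings under medium attenuation.
# Random nodes vary the waves and Modulators vary playback, with a wider pitch range on the smallest mix.
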SoundCue /BRCosmetics/Sounds/PickAxes/RoseLeader/PickAxe_RoseLeader_Swing
root=QualityLevel_1
nodes [
 {id=Attenuation_10, children [Modulator_6], properties {attenuationSettings=/Game/Sounds/Attenuation/Fort_Weapon_Pickaxe_Close_Att}}
 {id=Attenuation_11, children [Modulator_7], properties {attenuationSettings=/Game/Sounds/Attenuation/Fort_Weapon_Pickaxe_Medium_Att}}
 {id=Attenuation_6, children [Random_9], properties {attenuationSettings=&Fort_Weapon_Pickaxe_Close_Att}}
 {id=Attenuation_7, children [Random_5], properties {attenuationSettings=&Fort_Weapon_Pickaxe_Medium_Att}}
 {id=Attenuation_8, children [Random_10], properties {attenuationSettings=&Fort_Weapon_Pickaxe_Close_Att}}
 {id=Attenuation_9, children [Random_8], properties {attenuationSettings=&Fort_Weapon_Pickaxe_Medium_Att}}
 {id=Mixer_3, children [Attenuation_6, Attenuation_7], properties {inputVolume [1,1]}}
 {id=Mixer_4, children [Attenuation_8, Attenuation_9], properties {inputVolume [1,1]}}
 {id=Mixer_5, children [Attenuation_10, Attenuation_11], properties {inputVolume [1,1]}}
 {id=Modulator_0, children [Mixer_3], properties {pitchMax=1.03, pitchMin=0.97}}
 {id=Modulator_5, children [Mixer_4]}
 {id=Modulator_6, children [Random_11], properties {pitchMax=1.2, pitchMin=0.9}}
 {id=Modulator_7, children [WavePlayer_47], properties {pitchMax=1.2, pitchMin=0.9}}
 {id=QualityLevel_1, children [Modulator_0,Modulator_5,Mixer_5,Mixer_5]}
 {id=Random_10, children [WavePlayer_48, WavePlayer_49, WavePlayer_0], properties {weights [1,1,1]}}
 {id=Random_11, children [WavePlayer_51, WavePlayer_52], properties {weights [1,1]}}
 {
  id=Random_5
  children [WavePlayer_38, WavePlayer_39, WavePlayer_40, WavePlayer_41, WavePlayer_42, WavePlayer_43]
  properties {weights [1,1,1,1,1,1]}
 }
 {id=Random_8, children [WavePlayer_45, WavePlayer_44, WavePlayer_46], properties {weights [1,1,1]}}
 {
  id=Random_9
  children [WavePlayer_32, WavePlayer_33, WavePlayer_35, WavePlayer_36, WavePlayer_37, WavePlayer_1]
  properties {weights [1,1,1,1,1,1]}
 }
 {id=WavePlayer_0, properties {soundWaveAssetPtr=&/PickAxe_RoseLeader_Swing_04}}
 {id=WavePlayer_1, properties {soundWaveAssetPtr=&/PickAxe_RoseLeader_Swing_03}}
 {id=WavePlayer_32, properties {soundWaveAssetPtr=&/PickAxe_RoseLeader_Swing_01}}
 {id=WavePlayer_33, properties {soundWaveAssetPtr=&/PickAxe_RoseLeader_Swing_02}}
 {id=WavePlayer_35, properties {soundWaveAssetPtr=&PickAxe_RoseLeader_Swing_04}}
 {id=WavePlayer_36, properties {soundWaveAssetPtr=&/PickAxe_RoseLeader_Swing_05}}
 {id=WavePlayer_37, properties {soundWaveAssetPtr=&/PickAxe_RoseLeader_Swing_06}}
 {id=WavePlayer_38, properties {soundWaveAssetPtr=/Game/Athena/Sounds/Weapons/Fire_Axe_Swing_Distant_04}}
 {id=WavePlayer_39, properties {soundWaveAssetPtr=/Game/Athena/Sounds/Weapons/Fire_Axe_Swing_Distant_01}}
 {id=WavePlayer_40, properties {soundWaveAssetPtr=/Game/Athena/Sounds/Weapons/Fire_Axe_Swing_Distant_02}}
 {id=WavePlayer_41, properties {soundWaveAssetPtr=/Game/Athena/Sounds/Weapons/Fire_Axe_Swing_Distant_03}}
 {id=WavePlayer_42, properties {soundWaveAssetPtr=/Game/Athena/Sounds/Weapons/Fire_Axe_Swing_Distant_05}}
 {id=WavePlayer_43, properties {soundWaveAssetPtr=/Game/Athena/Sounds/Weapons/Fire_Axe_Swing_Distant_06}}
 {id=WavePlayer_44, properties {soundWaveAssetPtr=&Fire_Axe_Swing_Distant_03}}
 {id=WavePlayer_45, properties {soundWaveAssetPtr=&Fire_Axe_Swing_Distant_04}}
 {id=WavePlayer_46, properties {soundWaveAssetPtr=&Fire_Axe_Swing_Distant_06}}
 {id=WavePlayer_47, properties {soundWaveAssetPtr=&Fire_Axe_Swing_Distant_06}}
 {id=WavePlayer_48, properties {soundWaveAssetPtr=&PickAxe_RoseLeader_Swing_01}}
 {id=WavePlayer_49, properties {soundWaveAssetPtr=&PickAxe_RoseLeader_Swing_02}}
 {id=WavePlayer_51, properties {soundWaveAssetPtr=&PickAxe_RoseLeader_Swing_01}}
 {id=WavePlayer_52, properties {soundWaveAssetPtr=&PickAxe_RoseLeader_Swing_02}}
]
properties {
 audioPropertiesSheet=/BRCosmetics/Sounds/Settings/PropertySheets/Pickaxe/APS_Cosmetic_Pickaxe_Swing
 bOverrideConcurrency=true
 bPrimeOnLoad=true
 maxDistance=3000
 soundClassObject=&Fort_WeaponMelee_PSM
 volumeMultiplier=0.4
 concurrencyOverrides {bLimitToOwner=true, maxCount=3, resolutionRule=StopOldest}
}
